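# Niagara system for the idle effects of the VaneHush glider cosmetic, with front-end and gameplay tuning.
# Bone-pair anim-trail ribbons in gray, dark gray and gold switch by variant, tinted by time of day.
# Popcorn mesh/sprite bits and torus speed lines respond to forward movement.
# Variant-gated gold sparkles use baked offset arrays, with separate gameplay and front-end emitters; a wind-line emitter is disabled.
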NiagaraSystem /BRCosmetics/Gadgets/Assets/VinderTech_GliderChute/Glider_SilentNovel_VaneHush/FX/NS_Glider_SilentNovel_VaneHush_Idle
dependencies {
 usedDataInterfaces [
  &NiagaraDataInterfaceVector2DCurve
  &NiagaraDataInterfaceCurve
  &NiagaraDataInterfaceColorCurve
  &NiagaraDataInterfaceArrayFloat3
  &NiagaraDataInterfaceArrayFloat
 ]
 usedDynamicInputs [&Multiply_Vector2D_ByFloat]
 usedModules [
  &self:SystemSpawnScript.SystemScriptSource.SystemScriptGraph.NiagaraNodeAssignment_0.SetVariables_#1
  &FN_VariantManager
  &SystemState
  &self:SystemSpawnScript.SystemScriptSource.SystemScriptGraph.NiagaraNodeAssignment_1.SetVariables_#2
  &FN_Integrator
  &Timeline
 ]
}
emitters {
 DarkGrayTrail {
  emitterSpawnScript [&FN_RibbonIDOffsetUtility()]
  emitterUpdateScript [
   &EmitterState(
    Inactive Response=::NewEnumerator1
    Life Cycle Mode=::NewEnumerator1
    Loop Behavior=::NewEnumerator1
    Loop Duration Mode=::NewEnumerator1
    UseLoopDelay=false
   )
   &FN_SpawnAnimTrail(Number of Anim Trails:NiagaraInt32=&MakeIntFromFloat(Float=/BRCosmetics/Weapons/FORT_Melee/Pickaxe_GateHound/FX/VariantManager/VariantFloatSelector(
      Original Float:NiagaraFloat=2
      Variant 1 Float:NiagaraFloat=0
      Variant 2 Float:NiagaraFloat=?
      Variant 3 Float:NiagaraFloat=?
      Variant Amount:NiagaraInt32=0
     )))
  ]
  particleSpawnScript [
   &FN_InitializeAnimTrailBoneIndecies(First Bone or Socket Index=5, Second Bone or Socket Index=4)
   &FN_InitializeAnimTrailBoneIndecies(AnimTrail Selection=1, First Bone or Socket Index=7, Second Bone or Socket Index=6) {module=FN_InitializeAnimTrailBoneIndecies001}
   /SetVariables(Particles.AlphaFadeIn:NiagaraFloat=/Niagara/DynamicInputs/Math/Min_Float(A:NiagaraFloat=&Divide_Float(A=$Emitter.Age, B=1), B:NiagaraFloat=1)) {isSetParameters=true, module=SetVariables_#3}
   &FN_InitializeAnimTrail(
    Color=&MakeLinearColor(
     R:NiagaraFloat=1
     G:NiagaraFloat=1
     B:NiagaraFloat=1
     A:NiagaraFloat=&Multiply_Float(A=$Particles.AlphaFadeIn, B=0.5)
    )
    Lifetime=&MakeCustomFloatFromBool(Bool=$Particles.Locked, True Float=100, False Float=0.7)
    Position Offset=&FN_FrontEndVector3Modifier(Gameplay=<0 -20 0>, Front End=<0 -20 0>)
   )
   &AddVelocity(Velocity=&Multiply_VectorByFloat(Vector=<0 400 0>, Float=&MakeCustomFloatFromBool(Bool=$User.bIsFrontEnd, True Float=1, False Float=1)))
  ]
  particleUpdateScript [
   &SolveForcesAndVelocity(Clamp Velocity=false, Limit Acceleration=false)
   &FN_UpdateAnimTrail(
    Locked Particle Color:LinearColor=&MultiplyLinearColor(
     Linear Color=$Particles.Initial.Color
     Scale Factor=&MakeFloat4FromVectorAndFloat("Vector (RGB)"=<1 1 1>, "Float (Alpha)"=$System.FN_Integrator001.Value)
    )
    Locked Particle RibbonWidth=1
    SkeletalMesh=$User.SkeletalMesh
   ) {module=FN_UpdateAnimTrail001}
   /SetVariables(Particles.RibbonU0Override=$Particles.NormalizedAge) {isSetParameters=true, module=SetVariables_#4}
   &Color(
    Color=&FNTimeOfDayColor(
     Night Color=<rgba 0.0424 0.0435 0.05 1>
     Morning Color=<rgba 0.0962 0.1 0.0849 1>
     Day Color=<rgba 0.0962 0.1 0.0849 1>
     Evening Color=<rgba 0.0962 0.1 0.0849 1>
    )
    Scale Alpha=&Multiply_Float(
     A=&Multiply_Float(A=0.75, B=$Particles.AlphaFadeIn)
     B=&FN_FrontEndFloatModifier(Gameplay=$System.FN_Integrator001.Value, Front End=1)
    )
   )
  ]
  rendererClasses [&NiagaraRibbonRendererProperties]
  renderers [{rendererClass=&NiagaraRibbonRendererProperties}]
 }
 GoldTrails {
  emitterSpawnScript [&FN_RibbonIDOffsetUtility()]
  emitterUpdateScript [
   &EmitterState(
    Inactive Response=::NewEnumerator1
    Life Cycle Mode=::NewEnumerator1
    Loop Behavior=::NewEnumerator1
    Loop Duration Mode=::NewEnumerator1
    UseLoopDelay=false
   )
   &FN_SpawnAnimTrail(Number of Anim Trails:NiagaraInt32=&MakeIntFromFloat(Float=&VariantFloatSelector(
      Original Float:NiagaraFloat=0
      Variant 1 Float:NiagaraFloat=2
      Variant 2 Float:NiagaraFloat=?
      Variant 3 Float:NiagaraFloat=?
      Variant Amount:NiagaraInt32=0
     )))
  ]
  particleSpawnScript [
   &FN_InitializeAnimTrailBoneIndecies(First Bone or Socket Index=5, Second Bone or Socket Index=0)
   &FN_InitializeAnimTrailBoneIndecies(AnimTrail Selection=1, First Bone or Socket Index=7, Second Bone or Socket Index=2) {module=FN_InitializeAnimTrailBoneIndecies001}
   /SetVariables(Particles.AlphaFadeIn:NiagaraFloat=&Min_Float(A:NiagaraFloat=&Divide_Float(A=$Emitter.Age, B=1), B:NiagaraFloat=1)) {isSetParameters=true, module=SetVariables_#3}
   &FN_InitializeAnimTrail(
    AnimTrailWidth=&FN_FrontEndFloatModifier(Gameplay=1, Front End=0.82)
    Color=&MakeLinearColor(
     R:NiagaraFloat=1
     G:NiagaraFloat=1
     B:NiagaraFloat=1
     A:NiagaraFloat=&Multiply_Float(A=$Particles.AlphaFadeIn, B=0.5)
    )
    Lifetime=&MakeCustomFloatFromBool(Bool=$Particles.Locked, True Float=100, False Float=0.7)
    Position Offset=&FN_FrontEndVector3Modifier(Gameplay=<0 -20 0>, Front End=<0 -20 0>)
   )
   &AddVelocity(Velocity=&Multiply_VectorByFloat(Vector=<0 400 0>, Float=&MakeCustomFloatFromBool(Bool=$User.bIsFrontEnd, True Float=1, False Float=1)))
  ]
  particleUpdateScript [
   &SolveForcesAndVelocity(Clamp Velocity=false, Limit Acceleration=false)
   &FN_UpdateAnimTrail(
    Locked Particle Color:LinearColor=&MultiplyLinearColor(
     Linear Color=$Particles.Initial.Color
     Scale Factor=&MakeFloat4FromVectorAndFloat("Vector (RGB)"=<1 1 1>, "Float (Alpha)"=$System.FN_Integrator001.Value)
    )
    Locked Particle RibbonWidth=1
    SkeletalMesh=$User.SkeletalMesh
   ) {module=FN_UpdateAnimTrail001}
   /SetVariables(Particles.RibbonU0Override=$Particles.NormalizedAge) {isSetParameters=true, module=SetVariables_#4}
   &Color(
    Color=&FNTimeOfDayColor(
     Night Color=<rgba 0.2 0.104 0.02 1>
     Morning Color=<rgba 0.599 0.3115 0.0599 1>
     Day Color=<rgba 0.599 0.3115 0.0599 1>
     Evening Color=<rgba 0.599 0.3115 0.0599 1>
    )
    Scale Alpha=&Multiply_Float(
     A=&Multiply_Float(A=0.75, B=$Particles.AlphaFadeIn)
     B=&FN_FrontEndFloatModifier(Gameplay=$System.FN_Integrator001.Value, Front End=1)
    )
   )
  ]
  rendererClasses [&NiagaraRibbonRendererProperties]
  renderers [{rendererClass=&NiagaraRibbonRendererProperties}]
 }
 GrayTrail {
  emitterSpawnScript [&FN_RibbonIDOffsetUtility()]
  emitterUpdateScript [
   &EmitterState(
    Inactive Response=::NewEnumerator1
    Life Cycle Mode=::NewEnumerator1
    Loop Behavior=::NewEnumerator1
    Loop Duration Mode=::NewEnumerator1
    UseLoopDelay=false
   )
   &FN_SpawnAnimTrail(Number of Anim Trails:NiagaraInt32=&MakeIntFromFloat(Float=&VariantFloatSelector(
      Original Float:NiagaraFloat=2
      Variant 1 Float:NiagaraFloat=0
      Variant 2 Float:NiagaraFloat=?
      Variant 3 Float:NiagaraFloat=?
      Variant Amount:NiagaraInt32=0
     )))
  ]
  particleSpawnScript [
   &FN_InitializeAnimTrailBoneIndecies(First Bone or Socket Index=1, Second Bone or Socket Index=0)
   &FN_InitializeAnimTrailBoneIndecies(AnimTrail Selection=1, First Bone or Socket Index=3, Second Bone or Socket Index=2) {module=FN_InitializeAnimTrailBoneIndecies001}
   /SetVariables(Particles.AlphaFadeIn:NiagaraFloat=&Min_Float(A:NiagaraFloat=&Divide_Float(A=$Emitter.Age, B=1), B:NiagaraFloat=1)) {isSetParameters=true, module=SetVariables_#3}
   &FN_InitializeAnimTrail(
    AnimTrailWidth=&FN_FrontEndFloatModifier(Gameplay=1, Front End=0.8)
    Color=&MakeLinearColor(
     R:NiagaraFloat=1
     G:NiagaraFloat=1
     B:NiagaraFloat=1
     A:NiagaraFloat=&Multiply_Float(A=$Particles.AlphaFadeIn, B=0.5)
    )
    Lifetime=&MakeCustomFloatFromBool(Bool=$Particles.Locked, True Float=100, False Float=0.7)
    Position Offset=&FN_FrontEndVector3Modifier(Gameplay=<0 -20 0>, Front End=<0 -20 0>)
   )
   &AddVelocity(Velocity=&Multiply_VectorByFloat(Vector=<0 400 0>, Float=&MakeCustomFloatFromBool(Bool=$User.bIsFrontEnd, True Float=1, False Float=1)))
  ]
  particleUpdateScript [
   &SolveForcesAndVelocity(Clamp Velocity=false, Limit Acceleration=false)
   &FN_UpdateAnimTrail(
    Locked Particle Color:LinearColor=&MultiplyLinearColor(
     Linear Color=$Particles.Initial.Color
     Scale Factor=&MakeFloat4FromVectorAndFloat("Vector (RGB)"=<1 1 1>, "Float (Alpha)"=$System.FN_Integrator001.Value)
    )
    Locked Particle RibbonWidth=1
    SkeletalMesh=$User.SkeletalMesh
   ) {module=FN_UpdateAnimTrail001}
   /SetVariables(Particles.RibbonU0Override=$Particles.NormalizedAge) {isSetParameters=true, module=SetVariables_#4}
   &Color(
    Color=&VariantColorSelector(
     Variant Amount=0
     Original Color=&FNTimeOfDayColor(
      Night Color=<rgba 0.1503 0.1541 0.1771 1>
      Morning Color=<rgba 0.1703 0.1771 0.1503 1>
      Day Color=<rgba 0.1703 0.1771 0.1503 1>
      Evening Color=<rgba 0.1703 0.1771 0.1503 1>
     )
     Variant 1 Color=&FNTimeOfDayColor(
      Night Color=<rgba 0.9913 0.4362 0.1983 1>
      Morning Color=<rgba 0.599 0.3115 0.0599 1>
      Day Color=<rgba 0.599 0.3115 0.0599 1>
      Evening Color=<rgba 0.599 0.3115 0.0599 1>
     )
     Variant 2 Color=?
    )
    Scale Alpha=&Multiply_Float(
     A=&Multiply_Float(A=0.5, B=$Particles.AlphaFadeIn)
     B=&FN_FrontEndFloatModifier(Gameplay=$System.FN_Integrator001.Value, Front End=1)
    )
   )
  ]
  rendererClasses [&NiagaraRibbonRendererProperties]
  renderers [{rendererClass=&NiagaraRibbonRendererProperties}]
 }
 Popcorn {
  emitterUpdateScript [
   &EmitterState()
   &SpawnRate(
    Spawn Probability=0.4
    SpawnRate=&FN_FrontEndFloatModifier(Gameplay=&Multiply_Float(A=5, B=$User.ForwardDot), Front End=5)
    Use Spawn Probability=true
   )
   &SpawnBurst_Instantaneous(Spawn Count=75, Use Spawn Probability=false, UseLoopCountLimit=false) {enabled=false}
  ]
  particleSpawnScript [
   &InitializeParticle(
    Color=&FNTimeOfDayColor(
     Night Color=&Multiply_LinearColorByFloat(Linear Color=<rgba 1 0.7694 0.5677 1>, Float=1.5)
     Morning Color=<rgba 1 1 1 1>
     Day Color=<rgba 1 1 1 1>
     Evening Color=<rgba 1 1 1 1>
    )
    Lifetime Max=1.2
    Lifetime Min=0.8
    Lifetime Mode=::NewEnumerator1
    Mass Max=2
    Mass Mode=::NewEnumerator3
    Mesh Scale Mode=::NewEnumerator3
    Mesh Uniform Scale=&RandomRangeFloat(Minimum=0.8, Maximum=1.2)
    Sprite Size Max=<3 14>
    Sprite Size Min=<2 6>
    Sprite Size Mode=::NewEnumerator2
    UsePositionOffset=false
    Write Material Random=true
   )
   &AddVelocity(
    Cone Angle=50
    Cone Axis=<0 1 0.25>
    Distribution Along Cone Axis=0.2
    Module.UseConeFalloff=false
    Module.UseVelDistribution=true
    Override Seed=false
    Velocity Mode=::NewEnumerator2
    Velocity Speed=&RandomRangeFloat(Minimum=250, Maximum=750)
   ) {enabled=false}
   &AddVelocity(Velocity=<0 0 1>, Velocity Speed Scale=&RandomRangeFloat(Minimum=150, Maximum=200)) {module=AddVelocity001}
   &AddVelocity(Velocity=<0 1 0>, Velocity Speed Scale=&RandomRangeFloat(Minimum=150, Maximum=250)) {module=AddVelocity002}
   &InitialMeshOrientation(Use Rotation Vector=false)
   &SkeletalMeshLocation(Bone Sampling Mode=::NewEnumerator4, MaskSpawnGroup=false, Skeletal Mesh=$System.Skeletal Mesh Reader 2)
   &ShapeLocation(Disc Coverage=1, MaskSpawnGroup=false, Override Seed=false, Ring Radius=15, Shape Primitive=::NewEnumerator4, UseOwnerScale=true)
   &CameraOffset(Camera Offset Amount=1)
   /SetVariables() {isSetParameters=true}
  ]
  particleUpdateScript [
   &ParticleState()
   &ScaleMeshSize(Scale Factor=&Multiply_VectorByFloat(
     Vector=<1 1 1>
     Float=&FloatFromCurve(
      FloatCurve=DataInterface() {propertyValues {Curve {
         defaultValue=1
         keys [
          {arriveTangent=-0.0072, interpMode=RCIM_Cubic, leaveTangent=-0.0072, tangentMode=RCTM_User, time=0, value=0}
          {arriveTangentWeight=0.1274, interpMode=RCIM_Cubic, leaveTangentWeight=0.7778, tangentMode=RCTM_User, time=0.2, value=1}
          {arriveTangentWeight=0.4356, interpMode=RCIM_Cubic, tangentMode=RCTM_User, time=0.8, value=1}
          {arriveTangent=-0.0238, interpMode=RCIM_Cubic, leaveTangent=-0.0238, tangentMode=RCTM_User, time=1, value=0}
         ]
        }}}
      Scale Curve=0.65
     )
    ))
   &GravityForce()
   &Drag(Drag=&RandomRangeFloat(Minimum=0.8, Maximum=1.2), Use Linear Drag=true, Use Rotational Drag=false)
   &ScaleColor(
    Scale Alpha=&FloatFromCurve(FloatCurve=DataInterface() {propertyValues {Curve {keys [<at -0.0046 1.0113 RCIM_Cubic>,<at 1.0011 0.002 RCIM_Cubic>]}}})
    ScaleA=true
    ScaleRGB=true
   ) {enabled=false}
   &InheritSourceMovement(ApplyForce=false, ApplyPosition=true, ApplyPositionFalloff=false)
   &SolveForcesAndVelocity(Clamp Velocity=false, Limit Acceleration=false)
   &ScaleSpriteSizeBySpeed(
    Max Scale Factor=<1.2 3.3>
    Min Scale Factor=<0.75 0.3>
    Sample Scale Factor By Curve=true
    Scale Factor Curve=DataInterface() {propertyValues {Curve {keys [
        {arriveTangent=2.865, interpMode=RCIM_Cubic, time=0.1247, value=-0.0006}
        {arriveTangent=4.1522, interpMode=RCIM_Cubic, leaveTangent=4.1522, tangentMode=RCTM_User, time=0.3948, value=0.4893}
        <at 0.9958 0.9974 RCIM_Cubic>
       ]}}}
    Velocity Threshold=1500
   ) {enabled=false}
   &SkeletalMeshLocation(
    Bone Sampling Mode=::NewEnumerator4
    MaskSpawnGroup=false
    Position Sampling=::NewEnumerator2
    Sampled Position Offset=<0 0 20>
    Skeletal Mesh=$System.Initial.Skeletal Mesh Reader
   ) {enabled=false, module=SkeletalMeshLocation001}
   &SkeletalMeshLocation(Bone Sampling Mode=::NewEnumerator4, MaskSpawnGroup=false, Position Sampling=::NewEnumerator1, Skeletal Mesh=$System.Skeletal Mesh Reader) {enabled=false, module=SkeletalMeshLocation002}
   /SetVariables(Particles.Scale2:Vector3f=&Multiply_Vector(A=$Particles.Scale, B=&VectorFromFloat(Value=1.25))) {isSetParameters=true}
  ]
  rendererClasses [&NiagaraSpriteRendererProperties, &NiagaraMeshRendererProperties]
  renderers [
   {rendererClass=&NiagaraSpriteRendererProperties}
   {rendererClass=&NiagaraMeshRendererProperties, rendererIndex=1}
   {rendererClass=&NiagaraMeshRendererProperties, rendererIndex=2}
  ]
 }
 Sparkle {
  emitterUpdateScript [
   &EmitterState()
   &SpawnRate(
    Spawn Probability=0.5
    SpawnRate=&FN_FrontEndFloatModifier(
     Gameplay=&Multiply_Float(
      A=15
      B=&VariantFloatSelector(
       Original Float:NiagaraFloat=0
       Variant 1 Float:NiagaraFloat=1
       Variant 2 Float:NiagaraFloat=?
       Variant 3 Float:NiagaraFloat=?
       Variant Amount:NiagaraInt32=0
      )
     )
     Front End=0
    )
    Use Spawn Probability=true
   )
  ]
  particleSpawnScript [
   &InitializeParticle(
    Color=&FNTimeOfDayColor(
     Night Color=<rgba 0.9913 0.4362 0.1983 1>
     Morning Color=<rgba 1.5 0.6623 0.15 1>
     Day Color=<rgba 1.5 0.6623 0.15 1>
     Evening Color=<rgba 1.5 0.6623 0.15 1>
    )
    Lifetime=&RandomRangeFloat(Minimum=0.5, Maximum=1)
    Sprite Rotation Angle=0.25
    Sprite Rotation Mode=::NewEnumerator1
    Sprite Size Mode=::NewEnumerator1
    Uniform Sprite Size Max=12.5
    Uniform Sprite Size Min=7.5
    UsePositionOffset=false
    Write Material Random=true
   )
   /SetVariables(Particles.Position=&AddVectorToPosition(
     Position=&SelectVectorFromArray(
      Array Sampling Mode=::NewEnumerator0
      Direct Array Index=?
      Vector Selection Array=DataInterface() {propertyValues {FloatData [
         <-40.0099 -47.6794 65.4995>
         <27.6435 -50.8904 85.2878>
         <46.9658 -35.406 89.0246>
         <46.8121 50.4213 57.2297>
         <-15.3307 -31.0137 95.9323>
         <-12.0332 -4.775 23.7276>
         <-48.1236 -25.8535 20.8867>
         <-46.1887 -37.7743 92.7788>
         <-61.9054 28.2938 3.884>
         <30.533 -54.0331 105.316>
         <27.4795 -35.7594 8.0181>
         <7.8686 -54.8997 110.7045>
         <-16.2903 -41.1454 24.2344>
         <-47.5386 -27.4559 60.258>
         <6.7835 -23.8812 50.763>
         <-41.5072 -27.0999 71.1239>
         <15.8567 -48.815 72.2417>
         <40.6164 36.0539 59.1143>
         <-40.8022 -48.7905 72.5137>
         <37.422 -41.2227 33.6082>
         <20.5305 -26.2301 65.8763>
         <-34.1946 -24.0269 40.9574>
         <38.2774 0.597 58.351>
         <53.5236 50.8698 43.1195>
         <-33.4765 -51.8909 112.8702>
         <-25.1339 55.9292 29.5656>
         <31.8035 47.5881 28.8828>
         <-41.7962 -44.0272 104.3222>
         <32.1067 -5.2945 11.6588>
         <21.3503 -53.2535 100.0238>
         <-1.2131 -53.6856 102.7853>
         <55.5743 40.7047 50.8895>
         <-45.5719 -47.0308 64.3435>
         <-11.8893 -19.2354 21.0263>
         <44.1594 -37.4089 99.9511>
         <-2.9825 -41.5067 26.4959>
         <56.094 21.0062 40.2534>
         <-47.6597 -25.97 49.5104>
         <42.0366 13.2637 59.3361>
         <30.6265 -54.3474 107.3873>
         <-47.9036 -34.2547 76.7345>
         <-28.7394 -22.6704 42.8355>
         <-39.7069 -42.4074 20.3687>
         <16.6218 -30.6227 93.5654>
         <49.9266 14.5739 57.724>
         <-50.5846 37.9467 57.2998>
         <51.8893 -0.3994 34.4353>
         <-48.1754 -41.2601 51.9898>
         <-6.5142 -6.9549 22.9695>
         <44.5187 -50.0705 95.8343>
         <-47.4366 -46.6196 80.6106>
         <-47.2137 -27.5213 64.0743>
         <55.7635 8.5746 39.3111>
         <-34.925 -38.3596 32.2373>
         <-31.901 -4.0037 22.4098>
         <-47.5671 -45.6106 80.6769>
         <35.3929 -42.3894 20.9729>
         <-48.3996 -7.9631 30.6977>
         <-12.8845 -33.508 111.2952>
         <30.7093 -47.2694 62.7044>
         <-30.1176 2.9292 43.6823>
         <-2.4522 -34.8339 118.5212>
         <-46.0953 -43.9552 92.7136>
         <-7.7995 -33.4799 7.4691>
         <-39.4138 23.38 1.7403>
         <-40.6985 -29.8289 88.3027>
         <-4.6738 -41.1847 24.4803>
         <-39.8992 -27.4223 73.2383>
         <9.677 -49.3178 75.3891>
         <-10.0001 -21.5264 26.5168>
         <31.9983 14.7951 52.5138>
         <-56.4173 22.8387 48.4388>
         <31.7463 -14.3242 18.589>
         <55.4877 9.7757 51.0855>
         <19.9592 -50.7852 84.5738>
         <-0.3011 -52.4431 94.9515>
         <33.4001 -31.3557 98.0294>
         <-31.2745 -52.6793 96.6268>
         <8.1493 -52.2249 93.5859>
         <43.6133 -30.0152 87.5848>
         <-14.8994 34.0559 13.7716>
         <13.9649 -40.1257 121.6115>
         <-14.9486 1.2735 33.2503>
         <29.5856 -23.6798 49.4013>
         <-9.8296 -51.9178 91.6635>
         <5.3225 -41.6001 122.243>
         <-47.6668 50.2926 22.5477>
         <47.0444 -35.7821 88.5829>
         <-39.0738 -48.3648 69.7999>
         <43.8273 -6.7739 -1.3584>
         <-2.9627 -45.3713 50.6863>
         <-15.2938 -38.8186 121.5043>
         <36.4144 27.4342 57.2763>
         <-2.3405 20.3317 13.4349>
         <8.6911 37.5221 33.3298>
         <-25.69 -1.4997 32.9515>
         <31.8557 -25.057 77.5859>
         <-23.6496 13.0323 19.1712>
         <-10.4744 -4.6147 26.5291>
         <0.9427 13.1539 19.1>
        ]}}
      Random Seed=0
     )
     Vector=&FN_FrontEndVector3Modifier(Gameplay=<0 0 0>, Front End=<0 0 0>)
    )) {isSetParameters=true, module=SetVariables_#5}
   &CameraOffset(Camera Offset Amount=10)
  ]
  particleUpdateScript [
   &ParticleState()
   &ScaleSpriteSize(
    Uniform Curve Scale=5
    Uniform Curve Sprite Scale=DataInterface() {propertyValues {Curve {keys [
        {arriveTangent=4.7337, interpMode=RCIM_Cubic, leaveTangent=4.7337, tangentMode=RCTM_User, time=0, value=0}
        {interpMode=RCIM_Cubic, tangentMode=RCTM_User, time=0.5, value=1}
        {tangentMode=RCTM_User, time=1, value=0}
       ]}}}
   )
   &ScaleColor(
    Scale Alpha=&FloatFromCurve(FloatCurve=DataInterface() {propertyValues {Curve {keys [
         {arriveTangent=4.5234, interpMode=RCIM_Cubic, leaveTangent=4.5234, tangentMode=RCTM_User, time=0, value=0}
         {interpMode=RCIM_Cubic, tangentMode=RCTM_User, time=0.5, value=1}
         {tangentMode=RCTM_User, time=1, value=0}
        ]}}})
    Scale RGB=&Multiply_VectorByFloat(
     Vector=&FN_Select_Vector_from_Time_Of_Day(Morning=<1 1 1>, Day=<1 1 1>, Evening=<1 1 1>, Night=<0.5 0.8 2>)
     Float=&MakeFloatFromVector(
      Vector=&VectorFromFloat(Value=&FNTimeOfDayScalar(Night Scale=0.5, Morning Scale=1, Evening Scale=0.8))
      Channel=::NewEnumerator0
     )
    )
    ScaleA=true
    ScaleRGB=true
   )
   &SpriteRotationRate(Rotation Rate=&FloatFromCurve(
     FloatCurve=DataInterface() {propertyValues {Curve {keys [
         {interpMode=RCIM_Cubic, tangentMode=RCTM_User, time=0.2, value=0}
         {arriveTangent=3.4687, interpMode=RCIM_Cubic, leaveTangent=3.4687, tangentMode=RCTM_User, time=1, value=1}
        ]}}}
     Scale Curve=&SelectFloatFromArray(
      Array Sampling Mode=::NewEnumerator0
      Direct Array Index=?
      Float Selection Array=DataInterface() {propertyValues {FloatData [-800,800]}}
      Random Seed=0
     )
    ))
  ]
  rendererClasses [&NiagaraSpriteRendererProperties]
  renderers [{rendererClass=&NiagaraSpriteRendererProperties}]
 }
 Sparkle001 {
  emitterUpdateScript [
   &EmitterState()
   &SpawnRate(
    Spawn Probability=0.5
    SpawnRate=&FN_FrontEndFloatModifier(
     Gameplay=0
     Front End=&Multiply_Float(
      A=15
      B=&VariantFloatSelector(
       Original Float:NiagaraFloat=0
       Variant 1 Float:NiagaraFloat=1
       Variant 2 Float:NiagaraFloat=?
       Variant 3 Float:NiagaraFloat=?
       Variant Amount:NiagaraInt32=0
      )
     )
    )
    Use Spawn Probability=true
   )
  ]
  particleSpawnScript [
   &InitializeParticle(
    Color=&FNTimeOfDayColor(
     Night Color=<rgba 0.9913 0.4362 0.1983 1>
     Morning Color=<rgba 1.5 0.6623 0.15 1>
     Day Color=<rgba 1.5 0.6623 0.15 1>
     Evening Color=<rgba 1.5 0.6623 0.15 1>
    )
    Lifetime=&RandomRangeFloat(Minimum=0.5, Maximum=1)
    Sprite Rotation Angle=0.25
    Sprite Rotation Mode=::NewEnumerator1
    Sprite Size Mode=::NewEnumerator1
    Uniform Sprite Size Max=12.5
    Uniform Sprite Size Min=7.5
    UsePositionOffset=false
    Write Material Random=true
   )
   &CameraOffset(Camera Offset Amount=5)
  ]
  particleUpdateScript [
   &ParticleState()
   &SkeletalMeshLocation(
    Bone / Socket Index=1
    Bone Sampling Mode=::NewEnumerator4
    MaskSpawnGroup=false
    Position Sampling=::NewEnumerator2
    Skeletal Mesh=$System.Skeletal Mesh Reader 2
   )
   /SetVariables(Particles.Position=&AddVectorToPosition(
     Position=&AddVectorToPosition(
      Position=$Particles.SkeletalMeshLocation.SampledPosition
      Vector=&Multiply_VectorByFloat(
       Vector=&SelectVectorFromArray(
        Array Sampling Mode=::NewEnumerator0
        Direct Array Index=?
        Vector Selection Array=DataInterface() {propertyValues {FloatData [
           <-40.0099 -47.6794 65.4995>
           <27.6435 -50.8904 85.2878>
           <46.9658 -35.406 89.0246>
           <46.8121 50.4213 57.2297>
           <-15.3307 -31.0137 95.9323>
           <-12.0332 -4.775 23.7276>
           <-48.1236 -25.8535 20.8867>
           <-46.1887 -37.7743 92.7788>
           <-61.9054 28.2938 3.884>
           <30.533 -54.0331 105.316>
           <27.4795 -35.7594 8.0181>
           <7.8686 -54.8997 110.7045>
           <-16.2903 -41.1454 24.2344>
           <-47.5386 -27.4559 60.258>
           <6.7835 -23.8812 50.763>
           <-41.5072 -27.0999 71.1239>
           <15.8567 -48.815 72.2417>
           <40.6164 36.0539 59.1143>
           <-40.8022 -48.7905 72.5137>
           <37.422 -41.2227 33.6082>
           <20.5305 -26.2301 65.8763>
           <-34.1946 -24.0269 40.9574>
           <38.2774 0.597 58.351>
           <53.5236 50.8698 43.1195>
           <-33.4765 -51.8909 112.8702>
           <-25.1339 55.9292 29.5656>
           <31.8035 47.5881 28.8828>
           <-41.7962 -44.0272 104.3222>
           <32.1067 -5.2945 11.6588>
           <21.3503 -53.2535 100.0238>
           <-1.2131 -53.6856 102.7853>
           <55.5743 40.7047 50.8895>
           <-45.5719 -47.0308 64.3435>
           <-11.8893 -19.2354 21.0263>
           <44.1594 -37.4089 99.9511>
           <-2.9825 -41.5067 26.4959>
           <56.094 21.0062 40.2534>
           <-47.6597 -25.97 49.5104>
           <42.0366 13.2637 59.3361>
           <30.6265 -54.3474 107.3873>
           <-47.9036 -34.2547 76.7345>
           <-28.7394 -22.6704 42.8355>
           <-39.7069 -42.4074 20.3687>
           <16.6218 -30.6227 93.5654>
           <49.9266 14.5739 57.724>
           <-50.5846 37.9467 57.2998>
           <51.8893 -0.3994 34.4353>
           <-48.1754 -41.2601 51.9898>
           <-6.5142 -6.9549 22.9695>
           <44.5187 -50.0705 95.8343>
           <-47.4366 -46.6196 80.6106>
           <-47.2137 -27.5213 64.0743>
           <55.7635 8.5746 39.3111>
           <-34.925 -38.3596 32.2373>
           <-31.901 -4.0037 22.4098>
           <-47.5671 -45.6106 80.6769>
           <35.3929 -42.3894 20.9729>
           <-48.3996 -7.9631 30.6977>
           <-12.8845 -33.508 111.2952>
           <30.7093 -47.2694 62.7044>
           <-30.1176 2.9292 43.6823>
           <-2.4522 -34.8339 118.5212>
           <-46.0953 -43.9552 92.7136>
           <-7.7995 -33.4799 7.4691>
           <-39.4138 23.38 1.7403>
           <-40.6985 -29.8289 88.3027>
           <-4.6738 -41.1847 24.4803>
           <-39.8992 -27.4223 73.2383>
           <9.677 -49.3178 75.3891>
           <-10.0001 -21.5264 26.5168>
           <31.9983 14.7951 52.5138>
           <-56.4173 22.8387 48.4388>
           <31.7463 -14.3242 18.589>
           <55.4877 9.7757 51.0855>
           <19.9592 -50.7852 84.5738>
           <-0.3011 -52.4431 94.9515>
           <33.4001 -31.3557 98.0294>
           <-31.2745 -52.6793 96.6268>
           <8.1493 -52.2249 93.5859>
           <43.6133 -30.0152 87.5848>
           <-14.8994 34.0559 13.7716>
           <13.9649 -40.1257 121.6115>
           <-14.9486 1.2735 33.2503>
           <29.5856 -23.6798 49.4013>
           <-9.8296 -51.9178 91.6635>
           <5.3225 -41.6001 122.243>
           <-47.6668 50.2926 22.5477>
           <47.0444 -35.7821 88.5829>
           <-39.0738 -48.3648 69.7999>
           <43.8273 -6.7739 -1.3584>
           <-2.9627 -45.3713 50.6863>
           <-15.2938 -38.8186 121.5043>
           <36.4144 27.4342 57.2763>
           <-2.3405 20.3317 13.4349>
           <8.6911 37.5221 33.3298>
           <-25.69 -1.4997 32.9515>
           <31.8557 -25.057 77.5859>
           <-23.6496 13.0323 19.1712>
           <-10.4744 -4.6147 26.5291>
           <0.9427 13.1539 19.1>
          ]}}
        Random Seed=0
       )
       Float=0.9
      )
     )
     Vector=<0 -5 0>
    )) {isSetParameters=true, module=SetVariables_#5}
   &ScaleSpriteSize(
    Uniform Curve Scale=5
    Uniform Curve Sprite Scale=DataInterface() {propertyValues {Curve {keys [
        {arriveTangent=4.7337, interpMode=RCIM_Cubic, leaveTangent=4.7337, tangentMode=RCTM_User, time=0, value=0}
        {interpMode=RCIM_Cubic, tangentMode=RCTM_User, time=0.5, value=1}
        {tangentMode=RCTM_User, time=1, value=0}
       ]}}}
   )
   &ScaleColor(
    Scale Alpha=&FloatFromCurve(FloatCurve=DataInterface() {propertyValues {Curve {keys [
         {arriveTangent=4.5234, interpMode=RCIM_Cubic, leaveTangent=4.5234, tangentMode=RCTM_User, time=0, value=0}
         {interpMode=RCIM_Cubic, tangentMode=RCTM_User, time=0.5, value=1}
         {tangentMode=RCTM_User, time=1, value=0}
        ]}}})
    Scale RGB=&Multiply_VectorByFloat(
     Vector=&FN_Select_Vector_from_Time_Of_Day(Morning=<1 1 1>, Day=<1 1 1>, Evening=<1 1 1>, Night=<0.5 0.8 2>)
     Float=&MakeFloatFromVector(
      Vector=&VectorFromFloat(Value=&FNTimeOfDayScalar(Night Scale=0.5, Morning Scale=1, Evening Scale=0.8))
      Channel=::NewEnumerator0
     )
    )
    ScaleA=true
    ScaleRGB=true
   )
   &SpriteRotationRate(Rotation Rate=&FloatFromCurve(
     FloatCurve=DataInterface() {propertyValues {Curve {keys [
         {interpMode=RCIM_Cubic, tangentMode=RCTM_User, time=0.2, value=0}
         {arriveTangent=3.4687, interpMode=RCIM_Cubic, leaveTangent=3.4687, tangentMode=RCTM_User, time=1, value=1}
        ]}}}
     Scale Curve=&SelectFloatFromArray(
      Array Sampling Mode=::NewEnumerator0
      Direct Array Index=?
      Float Selection Array=DataInterface() {propertyValues {FloatData [-800,800]}}
      Random Seed=0
     )
    ))
  ]
  rendererClasses [&NiagaraSpriteRendererProperties]
  renderers [{rendererClass=&NiagaraSpriteRendererProperties}]
 }
 SpeedLine {
  emitterUpdateScript [
   &EmitterState()
   &SpawnRate(
    SpawnRate=&MakeCustomFloatFromBool(Bool=$User.bIsFrontEndPreview, True Float=50, False Float=90)
    Use Spawn Probability=true
   ) {module=SpawnRate001}
  ]
  particleSpawnScript [
   &InitializeParticle(
    Color=&MakeLinearColorFromVectorAndFloat(
     "Vector (RGB)"=<1 1 1>
     "Float (Alpha)"=&Multiply_Float(
      A=&UniformRangedFloat(Minimum=0.1, Maximum=0.18)
      B=&MakeCustomFloatFromBool(Bool=$User.bIsFrontEndPreview, True Float=0.5, False Float=1)
     )
    )
    Lifetime Max=0.16
    Lifetime Min=0.1
    Lifetime Mode=::NewEnumerator1
    Sprite Size Max=<3 10>
    Sprite Size Min=<2 2>
    Sprite Size Mode=::NewEnumerator2
    UsePositionOffset=false
    Write Material Random=true
   )
   &TorusLocation(
    Align Torus With Axis=true
    Handle Radius=0
    Large Radius:NiagaraFloat=&RandomRangeFloat(Minimum=120, Maximum=160)
    MaskSpawnGroup=false
    Module.Override Randomness=false
    Offset=<0 -180 0>
    Torus Axis=<0 1 0>
    Torus Origin:NiagaraPosition=&Add_Vector(
     A=Linked() {linkedVariable {name=Particles.Position, type=NiagaraPosition}}
     B=&MakeVector(
      X=0
      Y=&MakeCustomFloatFromBool(Bool=$User.bIsFrontEndPreview, True Float=-100, False Float=0)
      Z=&FN_FrontEndFloatModifier(Gameplay=0, Front End=-50)
     )
    )
   )
   &AddVelocityFromPoint(
    Module.Use Falloff Distance=false
    Origin Offset=<0 -500 0>
    Velocity Strength=&Multiply_Float(A=600, B=&MakeCustomFloatFromBool(Bool=$User.bIsFrontEndPreview, True Float=0.1, False Float=1))
   )
   &AddVelocity(Velocity=&Multiply_Vector(
     A=&RandomRangeVector(Minimum=<0 300 0>, Maximum=<0 700 0>)
     B=&MakeVector(X=1, Y=&MakeCustomFloatFromBool(Bool=$User.bIsFrontEndPreview, True Float=1, False Float=1), Z=1)
    ))
  ]
  particleUpdateScript [
   &SolveForcesAndVelocity(Clamp Velocity=false, Limit Acceleration=false)
   &ParticleState()
   &ScaleSpriteSize(Scale Factor=&Vector2DFromCurve(
     Vector2Curve=DataInterface() {propertyValues {XCurve {keys [<at 0 0.5 RCIM_Cubic>,<at 0.5 1 RCIM_Cubic>]}, YCurve {keys [<at 0 0 RCIM_Cubic>,<at 0.5 1 RCIM_Cubic>]}}}
     ScaleCurve=&Vector2DFromFloat(Value=1)
    ))
   &ScaleSpriteSizeBySpeed(Max Scale Factor=<1 50>, Min Scale Factor=<0.01 0.05>, Velocity Threshold=1)
   &ScaleColor(
    Scale Alpha=&FloatFromCurve(
     FloatCurve=DataInterface() {propertyValues {Curve {keys [
         <at 0 0 RCIM_Cubic>
         <at 0.3759 1.0207>
         {arriveTangent=-2.221, interpMode=RCIM_Cubic, leaveTangent=-2.221, tangentMode=RCTM_User, time=1, value=0}
        ]}}}
     Scale Curve=&MakeCustomFloatFromBool(
      Bool=$User.bIsFrontEndPreview
      True Float=&MakeCustomFloatFromBool(Bool=$User.bIsFullyDeployed, True Float=1, False Float=0)
      False Float=&Lerp_Float(A=0, B=0.66, Alpha=$User.ForwardDot)
     )
    )
    Scale RGB=&VectorFromFloat(Value=1)
    ScaleA=true
    ScaleRGB=true
   )
  ]
  rendererClasses [&NiagaraSpriteRendererProperties]
  renderers [{rendererClass=&NiagaraSpriteRendererProperties}]
 }
 WindLines {
  bEnabled=false
  emitterUpdateScript [
   &EmitterState()
   &SpawnRate(
    SpawnRate=&Lerp_Float(
     A=50
     B=150
     Alpha=&RemapRange(Input Value To Remap=$User.RightDot, Input Min=-1, Input Max=1, Output Min=0, Output Max=1)
    )
    Use Spawn Probability=true
   ) {enabled=false}
   &SpawnRate(SpawnGroup=1, SpawnRate=200, Use Spawn Probability=true) {module=SpawnRate001}
  ]
  particleSpawnScript [
   &InitializeParticle(
    Color=<rgba 1 1 1 5>
    Mesh Scale Max=<25 2 2>
    Mesh Scale Min=<10 1 1>
    Mesh Scale Mode=::NewEnumerator2
    UsePositionOffset=false
    Write Material Random=true
   )
   &InitialMeshOrientation(
    Mesh Orientation Mode=::NewEnumerator2
    Rotation=&RandomRangeVector(Minimum=<0 0 0>, Maximum=<1 0 0>)
    Use Rotation Vector=true
   )
   &TorusLocation(
    Align Torus With Axis=true
    Handle Radius=0
    Large Radius:NiagaraFloat=&RandomRangeFloat(Minimum=80, Maximum=150)
    MaskSpawnGroup=false
    Offset=<0 -180 0>
    Torus Axis=<0 1 0>
    Torus Distribution Mode:ENiagaraTorusDistributionMode=::NewEnumerator1
    Torus Origin:NiagaraPosition=&Add_Vector(
     A=Linked() {linkedVariable {name=Particles.Position, type=NiagaraPosition}}
     B=&MakeVector(X=0, Y=&MakeCustomFloatFromBool(Bool=$User.bIsFrontEndPreview, True Float=200, False Float=0), Z=0)
    )
    U Position:NiagaraFloat=&MakeCustomFloatFromBool(
     Bool=&SetBoolByFloatComparison(A=$User.RightDot, B=0, Comparison Type=::NewEnumerator4)
     True Float=&RandomRangeFloat(Minimum=0, Maximum=1)
     False Float=&Lerp_Float(
      A=&RandomRangeFloat(Minimum=0.25, Maximum=0.75)
      B=&RandomRangeFloat(Minimum=-0.25, Maximum=0.25)
      Alpha=&RemapRange(Input Value To Remap=$User.RightDot, Input Min=-1, Input Max=1, Output Min=0, Output Max=1)
     )
    )
    V Position:NiagaraFloat=&RandomRangeFloat(Minimum=0, Maximum=1)
   ) {enabled=false}
   &TorusLocation(
    Align Torus With Axis=true
    Handle Radius=0
    Large Radius:NiagaraFloat=&RandomRangeFloat(Minimum=100, Maximum=120)
    MaskSpawnGroup=false
    Module.Override Randomness=false
    Offset=<0 -180 0>
    Spawn Group Mask=1
    Torus Axis=<0 1 0>
    Torus Origin:NiagaraPosition=&Add_Vector(A=<0 -100 0>, B=&Lerp_Vector(A=<0 0 0>, B=<200 0 0>, Alpha=$User.RightDot))
   ) {module=TorusLocation001}
   &AddVelocityFromPoint(
    Module.Use Falloff Distance=false
    Origin Offset=&Add_Vector(A=<0 -500 0>, B=&Lerp_Vector(A=<0 0 0>, B=<300 0 0>, Alpha=$User.RightDot))
    Velocity Strength=150
   )
   &AddVelocity(
    Module.Use Falloff Distance=false
    Origin Offset=&Add_Vector(A=<0 -500 0>, B=&Lerp_Vector(A=<0 0 0>, B=<500 0 0>, Alpha=$User.RightDot))
    Override Seed=false
    Velocity Mode=::NewEnumerator1
    Velocity Speed=200
   ) {module=AddVelocity001}
   &AddVelocity(Velocity=&Multiply_Vector(
     A=&RandomRangeVector(Minimum=<1 200 1>, Maximum=<1 300 1>)
     B=&MakeVector(X=1, Y=&MakeCustomFloatFromBool(Bool=$User.bIsFrontEndPreview, True Float=-1, False Float=1), Z=1)
    ))
  ]
  particleUpdateScript [
   &InheritSourceMovement(Applied Position Scale=<0.9 0.9 1>, ApplyForce=false, ApplyPosition=true, ApplyPositionFalloff=false) {enabled=false}
   &ScaleVelocity(Velocity Scale=&VectorFromFloat(Value=&FloatFromCurve(
      FloatCurve=DataInterface() {propertyValues {Curve {keys [
          {interpMode=RCIM_Cubic, tangentMode=RCTM_User, time=0, value=0}
          {arriveTangent=2, interpMode=RCIM_Cubic, leaveTangent=2, time=0.1, value=0.975}
          {interpMode=RCIM_Cubic, tangentMode=RCTM_User, time=0.5, value=1}
          {arriveTangent=-2, interpMode=RCIM_Cubic, leaveTangent=-2, time=0.9, value=1}
          {tangentMode=RCTM_User, time=1, value=0}
         ]}}}
      CurveIndex=&RemapRange(Input Value To Remap=$User.RightDot, Input Min=-1, Input Max=1, Output Min=0, Output Max=1)
     )))
   &SolveForcesAndVelocity(Clamp Velocity=false, Limit Acceleration=false)
   &ParticleState()
   &ScaleMeshSize(Scale Factor=&VectorFromFloat(Value=&FloatFromCurve(FloatCurve=DataInterface() {propertyValues {Curve {keys [
          {tangentMode=RCTM_User, time=0, value=1}
          {arriveTangent=-0.0405, interpMode=RCIM_Cubic, leaveTangent=-0.0405, tangentMode=RCTM_User, time=0.7009, value=1}
          {arriveTangent=-8.0089, interpMode=RCIM_Cubic, leaveTangent=-8.0089, tangentMode=RCTM_User, time=1, value=0}
         ]}}})))
   &DynamicMaterialParameters(
    Index 0 Param 2=5
    Index 0 Param 3=1
    Index 0 Param 4=1
    Module.Param0WriteEnabled=true
    Module.Param1WriteEnabled=true
    Module.Param2WriteEnabled=true
    Module.Param3WriteEnabled=true
   )
   &ScaleColor(
    Scale Alpha=&FloatFromCurve(
     FloatCurve=DataInterface() {propertyValues {Curve {keys [
         {interpMode=RCIM_Cubic, tangentMode=RCTM_User, time=0, value=0}
         {interpMode=RCIM_Cubic, tangentMode=RCTM_User, time=0.5, value=1.5}
         {tangentMode=RCTM_User, time=1, value=0}
        ]}}}
     Scale Curve=&Lerp_Float(A=0, B=&Lerp_Float(A=0, B=1, Alpha=$System.Timeline001.Time), Alpha=$System.Timeline.Time)
    )
    ScaleA=true
    ScaleRGB=true
   )
   &Color(Color=&ColorFromCurve(
     ColorCurve=DataInterface() {propertyValues {
       AlphaCurve {keys [<at 0 0 RCIM_Cubic>,<at 0.0502 1>,<at 0.4858 1>,<at 0.9519 1>,<at 1 0 RCIM_Cubic>]}
       BlueCurve {keys [<at 0 1 RCIM_Cubic>,<at 0.4934 0.0763>,<at 1 0 RCIM_Cubic>]}
       GreenCurve {keys [<at 0 0.0001 RCIM_Cubic>,<at 0.4934 0>,<at 1 1 RCIM_Cubic>]}
       RedCurve {keys [<at 0 0 RCIM_Cubic>,<at 0.4934 1>,<at 1 0.0424 RCIM_Cubic>]}
      }}
     CurveIndex=&RemapRange(Input Value To Remap=$User.RightDot, Input Min=-1, Input Max=1, Output Min=0, Output Max=1)
    )) {enabled=false}
  ]
  rendererClasses [&NiagaraMeshRendererProperties, &NiagaraSpriteRendererProperties]
  renderers [
   {rendererClass=&NiagaraMeshRendererProperties, rendererIndex=0}
   {rendererClass=&NiagaraSpriteRendererProperties, rendererIndex=1}
  ]
 }
}
summary {
 systemName=NS_Glider_SilentNovel_VaneHush_Idle
 emitters [
  {emitterName=SpeedLine, rendererClasses [&NiagaraSpriteRendererProperties]}
  {
   bEnabled=false
   emitterName=WindLines
   rendererClasses [&NiagaraMeshRendererProperties, &NiagaraSpriteRendererProperties]
  }
  {emitterName=Sparkle, rendererClasses [&NiagaraSpriteRendererProperties]}
  {emitterName=Popcorn, rendererClasses [&NiagaraSpriteRendererProperties, &NiagaraMeshRendererProperties]}
  {emitterName=GrayTrail, rendererClasses [&NiagaraRibbonRendererProperties]}
  {emitterName=DarkGrayTrail, rendererClasses [&NiagaraRibbonRendererProperties]}
  {emitterName=GoldTrails, rendererClasses [&NiagaraRibbonRendererProperties]}
  {emitterName=Sparkle001, rendererClasses [&NiagaraSpriteRendererProperties]}
 ]
 userVariables [
  User.bIsFrontEnd {type=NiagaraBool}
  User.bIsFrontEndPreview {type=NiagaraBool}
  User.bIsFullyDeployed {type=NiagaraBool}
  User.ForwardDot {type=NiagaraFloat, defaultValue=0, description=""}
  User.RibbonIDOffset {type=NiagaraInt32}
  User.RightDot {type=NiagaraFloat, defaultValue=0, description=""}
  User.SkeletalMesh {
   type=NiagaraDataInterfaceSkeletalMesh
   defaultValue {
    struct=NiagaraExt_VariableValue_DataInterface
    value {dataInterface=&self:NiagaraDataInterfaceSkeletalMesh_2}
   }
  }
  User.Variant {type=NiagaraFloat, defaultValue=1, description=""}
 ]
}
